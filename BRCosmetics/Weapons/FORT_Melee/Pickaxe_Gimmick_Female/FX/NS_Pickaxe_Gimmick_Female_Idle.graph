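# Idle Niagara effect for the Gimmick Female pickaxe that places three blue glow points on the weapon.
# A single Light emitter burst-spawns three particles, each drawn by both a sprite and a light renderer.
# Spawn groups index a vector array to give each particle its own fixed position offset.
# Particles ignore lifetime expiry, with infinitely lived ones allowed to die when the emitter deactivates.
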NiagaraSystem /BRCosmetics/Weapons/FORT_Melee/Pickaxe_Gimmick_Female/FX/NS_Pickaxe_Gimmick_Female_Idle
dependencies {
 usedDataInterfaces [&NiagaraDataInterfaceArrayFloat3]
 usedDynamicInputs []
 usedModules [&SystemState]
}
emitters {Light {
  emitterUpdateScript [
   &EmitterState(Life Cycle Mode=::NewEnumerator1, Loop Behavior=::NewEnumerator1, Loop Duration Mode=::NewEnumerator1, UseLoopDelay=false)
   &SpawnBurst_Instantaneous(Spawn Count=3, Use Spawn Probability=false, UseLoopCountLimit=false)
  ]
  particleSpawnScript [
   &FN_SetSpawnGroup(Number of Spawn Groups=3)
   &InitializeParticle(
    Color=<rgba 0 0.5833 1 1>
    Position Offset=&SelectVectorFromArray(
     Direct Array Index=$Particles.SpawnGroup
     Vector Selection Array=DataInterface() {propertyValues {FloatData [<0 30 60>, <0 -20 60>, <0 5 70>]}}
    )
    Sprite Size Mode=::NewEnumerator3
    UsePositionOffset=true
    Write Material Random=true
   )
  ]
  particleUpdateScript [&ParticleState(Kill Particles When Lifetime Has Elapsed=false, Let Infinitely Lived Particles Die When Emitter Deactivates=true)]
  rendererClasses [&NiagaraSpriteRendererProperties, &NiagaraLightRendererProperties]
  renderers [
   {rendererClass=&NiagaraSpriteRendererProperties}
   {rendererClass=&NiagaraLightRendererProperties, rendererIndex=1}
  ]
 }}
summary {
 systemName=NS_Pickaxe_Gimmick_Female_Idle
 emitters [{emitterName=Light, rendererClasses [&NiagaraSpriteRendererProperties, &NiagaraLightRendererProperties]}]
 userVariables []
}
