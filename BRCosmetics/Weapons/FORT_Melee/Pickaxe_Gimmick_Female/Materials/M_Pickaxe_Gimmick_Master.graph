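# Master material for the Gimmick pickaxe that adds animated energy to the emissive output of the shared character shader chain.
# Two scrolling polar noise layers, centered on an offset pre-skinned position and warped by curl noise, blend by a radial center mask.
# Energy tint, a periodic radial pulse and summed-sine flicker are masked by the emissive blue channel.
# Noise samples take derivatives from planar coordinates, bypassing the polar angle seam.
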Material /BRCosmetics/Weapons/FORT_Melee/Pickaxe_Gimmick_Female/Materials/M_Pickaxe_Gimmick_Master
graph {
 properties {
  bAutomaticallySetUsageInEditor=false
  bUseMaterialAttributes=true
  bUsedWithClothing=true
  bUsedWithMeshDeformer=true
  bUsedWithMorphTargets=true
  bUsedWithSkeletalMesh=true
  pixelDepthOffsetMode=PDOM_Legacy
  refractionMethod=RM_IndexOfRefraction
  shadingModel=MSM_FromMaterialExpression
  subsurfaceProfile=&SS_HeroSkin_02
  usedShadingModels=Default Lit | Subsurface | Clear Coat | Subsurface Profile | Cloth | Eye
 }
 nodes [
  n1 = MaterialFunctionCall() {materialFunction=&BaseFN_CharacterFunction_PreFX}
  n2 = GetMaterialAttributes(None=MaterialFunctionCall(
    PreskinnedPosition=NamedRerouteDeclaration(None=@n1.Pre skinned Local Position) {name=PreSkinned_Location, nodeColor=<rgba 0 0.5781 1 1>}
    Gmap=@n1.Gmap
    MaterialAttributes=@n1.Gmap
    AO=@n1.Gmap
   ) {materialFunction=&CharacterSkinFX}.Result) {attributeGetTypes [#1, #2]}
  n3 = ComponentMask(None=@n2.Emissive Color) {b=true}
  n4 = Subtract(A=@n3, B=Constant() {r=0.8}) {constB=0.9}
  n5 = Constant() {r=0.75}
  n6 = Time()
  n7 = Multiply(A=@n6) {constB=3.333}
  n8 = Add(A=Add(A=Cosine(None=@n6), B=Sine(None=@n7)), B=Sine(None=Multiply(A=@n6) {constB=7.777}))
  n9 = Multiply(A=@n3, B=@n8) {constB=50}
  n10 = Multiply(A=@n8, B=ScalarParameter() {defaultValue=0.0075, parameterName=Flicker_Intensity}) {constB=0.01}
  n11 = TextureSample(UVs=Divide(A=NamedRerouteUsage(), B=ScalarParameter() {defaultValue=30, group=Gimmick, parameterName=CurlSize})) {samplerSource=SSM_Wrap_WorldGroupSettings, samplerType=SAMPLERTYPE_LinearColor, texture=/Game/Effects/FogVolume/VT_CurlNoise_LOW}
  n12 = Multiply(A=ComponentMask(None=@n11.RGB) {b=true, g=true}, B=ScalarParameter() {defaultValue=0.005, group=Gimmick, parameterName=CurlStrength})
  n13 = ScalarParameter() {defaultValue=0.5, group=Gimmick, parameterName=Noise_Center_Raidal_Tile}
  n14 = Constant() {r=-0.5}
  n15 = ComponentMask(None=Add(A=NamedRerouteUsage(), B=VectorParameter() {parameterName=CenterOffset, defaultValue=<rgba 0 0 -53 0>}.RGB)) {b=true, g=true}
  n16 = ScalarParameter() {defaultValue=15, group=Gimmick, parameterName=CenterNoise_Size}
  n17 = Divide(A=@n15, B=@n16) {constB=25}
  n18 = Add(A=AppendVector(A=@n14, B=@n14), B=@n17)
  n19 = ComponentMask(None=@n18) {r=true}
  n20 = Multiply(A=@n13, B=@n19)
  n21 = Multiply(A=@n20, B=@n20)
  n22 = ComponentMask(None=@n18) {g=true}
  n23 = Multiply(A=@n13, B=@n22)
  n24 = Multiply(A=@n23, B=@n23)
  n25 = ScalarParameter() {defaultValue=1, group=Gimmick, parameterName=Noise_Center_Angle_Tile}
  n26 = AppendVector(
   A=SquareRoot(None=Add(A=@n21, B=@n24))
   B=Divide(
    A=Arctangent2Fast(Y=@n22, X=@n19)
    B=MaterialFunctionCall(Multiplier=@n25) {desc=Radial Gradient, materialFunction=&Pi}.Result
   )
  )
  n27 = ScalarParameter() {defaultValue=0.1, group=Gimmick, parameterName=EnergySpeed}
  n28 = MaterialFunctionCall(X=Multiply(A=Time(), B=Multiply(A=@n27) {constB=-1})) {materialFunction=&MakeFloat2}
  n29 = MaterialFunctionCall(Float2=@n17) {materialFunction=&BreakOutFloat2Components}
  n30 = DDX(Value=ComponentMask(None=@n29.R) {r=true})
  n31 = DDY(Value=ComponentMask(None=@n29.G) {g=true})
  n32 = TextureSample!(UVs=Add(A=@n12, B=Add(A=@n26, B=@n28.Result) {constB=0.5}), Tex, "DDX(UVs)"=@n30, "DDY(UVs)"=@n31, Apply View MipBias) {
   automaticViewMipBias=false
   constMipValue=-2
   mipValueMode=TMVM_Derivative
   samplerSource=SSM_Wrap_WorldGroupSettings
   texture=/Game/Weapons/WeaponSkins/Textures/T_Swirly_Noise_highContrast
  }
  n33 = Multiply(A=@n32.G) {constB=0.75}
  n34 = ScalarParameter() {defaultValue=0.2, group=Gimmick, parameterName=Noise_Edge_Radial_Tile}
  n35 = Constant() {r=-0.5}
  n36 = Add(A=AppendVector(A=@n35, B=@n35), B=@n17)
  n37 = ComponentMask(None=@n36) {r=true}
  n38 = Multiply(A=@n34, B=@n37)
  n39 = ComponentMask(None=@n36) {g=true}
  n40 = Multiply(A=@n34, B=@n39)
  n41 = Add(A=Multiply(A=@n38, B=@n38), B=Multiply(A=@n40, B=@n40))
  n42 = Arctangent2Fast(Y=@n39, X=@n37)
  n43 = MaterialFunctionCall(Multiplier=ScalarParameter() {defaultValue=0.5, group=Gimmick, parameterName=Noise_Edge_Angle_Tile}) {desc=Radial Gradient, materialFunction=&Pi}
  n44 = Add(A=AppendVector(A=SquareRoot(None=@n41), B=Divide(A=@n42, B=@n43.Result))) {constB=0.333}
  n45 = TextureSample(UVs=Add(A=Add(A=@n44, B=@n28.Result), B=@n12), "DDX(UVs)"=@n30, "DDY(UVs)"=@n31) {
   automaticViewMipBias=false
   constMipValue=-2
   mipValueMode=TMVM_Derivative
   samplerSource=SSM_Wrap_WorldGroupSettings
   samplerType=SAMPLERTYPE_LinearGrayscale
   texture=/BRCosmetics/Weapons/FORT_Melee/Pickaxe_Cube_Ninja_Male/FX/T_FluidFlameTile
  }
  n46 = Multiply(A=@n45.G) {constB=1.5}
  n47 = NamedRerouteUsage()
  n48 = Add(A=ComponentMask(None=@n26) {r=true}) {constB=6}
  n49 = Time() {bOverride_Period=true, period=14}
  n50 = Saturate(None=Multiply(A=Subtract(A=@n48, B=@n49)) {constB=3})
  n51 = Constant()
  n52 = Add(
   A=Multiply(
    A=LinearInterpolate(A=@n33, B=@n46, Alpha=@n47)
    B=VectorParameter() {group=Gimmick, parameterName=Energy_Color, defaultValue=<rgba 0 0.4031 0.625 0>}.RGB
   )
   B=Multiply(
    A=VectorParameter() {group=Gimmick, parameterName=Center_Pulse_Color, defaultValue=<rgba 0 0.6042 0.3409 0>}.RGB
    B=If(A=@n50, "A > B"=@n51, "A < B"=@n50) {constB=1}
   )
  )
  n53 = Multiply(
   A=OneMinus(None=MaterialFunctionCall!(In=@n4, Ramp Width=@n5) {materialFunction=/Engine/Functions/Engine_MaterialFunctions01/Math/SmoothCeil}.Result)
   B=Add(A=Multiply(A=@n9, B=@n10), B=Multiply(A=@n3, B=@n52))
  )
  n54 = MaterialFunctionCall(
   AO=@n1.AO
   Distance Blend=@n1.AO
   Fresnel=@n1.AO
   Gmap=@n1.AO
   Pre Skinned Local Position=@n1.AO
   Material=SetMaterialAttributes(MaterialAttributes=@n2.MaterialAttributes, Emissive Color=Add(A=@n2.Emissive Color, B=@n53)) {attributeSetTypes [#1]}
  ) {materialFunction=&BaseFN_CharacterFunction_PostFX}
  MaterialFunctionCall(In=@n54.Result) {materialFunction=&CINE_MF_CharPostFX}
  n55 = Multiply(A=@n16) {constB=-0.5}
  n56 = Divide(
   A=MaterialFunctionCall(Vector 2=Add(A=@n15, B=@n55)) {materialFunction=/Engine/Functions/Engine_MaterialFunctions02/Utility/VectorLength}.V2 Length
   B=@n16
  )
  NamedRerouteDeclaration(None=Floor(None=Saturate(None=@n56))) {name=Center_Mask, nodeColor=<rgba 0.5703 1 0 1>}
 ]
}
