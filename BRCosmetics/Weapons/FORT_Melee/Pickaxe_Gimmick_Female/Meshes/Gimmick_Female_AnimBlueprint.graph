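# Animation Blueprint for the female Gimmick pickaxe that adds secondary motion to its front and back axe blade bones.
# Each update, two Random Vector Function calls with differing frequencies produce time-driven offsets, added to the blades as bone-space translation.
# Spring-based Anim Dynamics nodes then simulate each blade relative to b_AR_Root within tight linear limits.
# Jump Alpha sets the dynamics weight from bIsOnGround, full on the ground and reduced in the air.
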AnimBlueprint /BRCosmetics/Weapons/FORT_Melee/Pickaxe_Gimmick_Female/Meshes/Gimmick_Female_AnimBlueprint
target_skeleton=&/Gimmick_Female_Skeleton
graphs [
 AnimGraph {nodes [
   n1 = "Transform (Modify) Bone - Bone: dyn_axeBlade_front"(
    ComponentPose=Local To Component(LocalPose=Input Pose())
    Alpha=1
    Translation=Get Blade Fr Motion() -> (Blade Fr Motion:Vector)
    Rotation="0,0,0"
    Scale="1,1,1"
   ) {
    rotationSpace=BCS_WorldSpace
    translationMode=BMM_Additive
    translationSpace=BCS_BoneSpace
    alphaBoolBlend {blendOption=Linear}
    alphaScaleBiasClamp {interpSpeedDecreasing=2, interpSpeedIncreasing=2}
    boneToModify {boneName=dyn_axeBlade_front}
   }
   n2 = "Transform (Modify) Bone - Bone: dyn_axeBlade_back"(ComponentPose=@n1, Alpha=1, Translation=Get Blade Bk Motion() -> (Blade Bk Motion:Vector), Rotation="0,0,0", Scale="1,1,1") {
    rotationSpace=BCS_WorldSpace
    translationMode=BMM_Additive
    translationSpace=BCS_BoneSpace
    alphaBoolBlend {blendOption=Linear}
    alphaScaleBiasClamp {interpSpeedDecreasing=2, interpSpeedIncreasing=2}
    boneToModify {boneName=dyn_axeBlade_back}
   }
   n3 = Get Jump Alpha() -> (Jump Alpha:double)
   n4 = Anim Dynamics - Bone: dyn_axeBlade_front(ComponentPose=@n2, Alpha=@n3, ExternalForce="0,0,0") {
    angularDampingOverride=0.9
    angularSpringConstant=100
    bAngularSpring=true
    bGravityOverrideInSimSpace=true
    bLinearSpring=true
    bOverrideAngularDamping=true
    bOverrideLinearDamping=true
    bUseGravityOverride=true
    gravityScale=0
    linearDampingOverride=0.9999
    linearSpringConstant=120
    alphaBoolBlend {blendOption=Linear}
    alphaScaleBiasClamp {bInterpResult=true}
    boundBone {boneName=dyn_axeBlade_front}
    componentAppliedLinearAccClamp=<3000 3000 3000>
    componentLinearAccScale=<0.05 0.05 0.05>
    componentLinearVelScale=<0.9 0.9 0.9>
    gravityOverride {z=-300}
    physicsBodyDefinitions [{
      sphereCollisionRadius=0
      boundBone {boneName=dyn_axeBlade_front}
      boxExtents=<75 75 75>
      constraintSetup {angularLimitsMax=<0 0 0>, angularLimitsMin=<0 0 0>, angularTarget=<0 0 0>, linearAxesMax=<2 3 1>, linearAxesMin=<-1 -1 -0.5>}
      localJointOffset=<0 0 0>
     }]
    relativeSpaceBone {boneName=b_AR_Root}
   }
   n5 = Anim Dynamics - Bone: dyn_axeBlade_back(ComponentPose=@n4, Alpha=@n3, ExternalForce="0,0,0") {
    angularDampingOverride=0.9
    angularSpringConstant=100
    bAngularSpring=true
    bGravityOverrideInSimSpace=true
    bLinearSpring=true
    bOverrideAngularDamping=true
    bOverrideLinearDamping=true
    bUseGravityOverride=true
    gravityScale=0
    linearDampingOverride=0.9999
    linearSpringConstant=210
    alphaBoolBlend {blendOption=Linear}
    alphaScaleBiasClamp {bInterpResult=true}
    boundBone {boneName=dyn_axeBlade_back}
    componentAppliedLinearAccClamp=<3000 3000 3000>
    componentLinearAccScale=<0.03 0.03 0.03>
    componentLinearVelScale=<0.9 0.9 0.9>
    gravityOverride {z=-300}
    physicsBodyDefinitions [{
      sphereCollisionRadius=0
      boundBone {boneName=dyn_axeBlade_back}
      boxExtents=<75 75 75>
      constraintSetup {angularLimitsMax=<0 0 0>, angularLimitsMin=<0 0 0>, angularTarget=<0 0 0>, linearAxesMax=<2 1 1>, linearAxesMin=<-1 -3 -0.5>}
      localJointOffset=<0 0 0>
     }]
    relativeSpaceBone {boneName=b_AR_Root}
   }
   Output Pose(Result=Component To Local(ComponentPose=@n5)) {layerGroup=None}
  ]}
 EventGraph {nodes [
   Event BlueprintUpdateAnimation()
   > n1 = Sequence!(execute) -> (then_0, then_1, then_2:Exec)
   n2 = GetGameTimeInSeconds()
   n3 = GetOwnerSkeletalMeshComponent()
   > n4 = Random Vector Function!(
    execute
    CharacterPartAnimInstance:Custom Character Part Anim Instance Object Reference
    SkelMeshComponent:Mesh Component Object Reference=@n3
    JointName:Name=None
    GameTimeInSeconds:double=@n2
    TimeOffset:double=0
    PerlinScale:double=0
    Frequency:Vector="1,1.3,0.8"
    Amplitude:Vector="0.1,1,0.6"
    Offset:Vector="0,0,0"
   ) -> (then, VectorOut:Vector)
   > Set Blade Fr Motion(Blade Fr Motion:Vector=@n4) -> (Output_Get:Vector)
   > n5 = Random Vector Function(
    SkelMeshComponent=@n3
    JointName=None
    GameTimeInSeconds=@n2
    TimeOffset=0
    PerlinScale=0
    Frequency="0.8,0.9,1.2"
    Amplitude="0.1,1,0.6"
    Offset="0,0,0"
   )
   > Set Blade Bk Motion(Blade Bk Motion:Vector=@n5) -> (Output_Get:Vector)
   n6 = Get bIsOnGround() -> (bIsOnGround:bool)
   Set Jump Alpha(execute=@n1.then_1, Jump Alpha:double=SelectFloat(A=0.65, B=1, bPickA=NOT Boolean(A=@n6))) -> (Output_Get:double)
   TryGetPawnOwner()
  ]}
]
variables [
 Blade Bk Motion {type=Vector}
 Blade Fr Motion {type=Vector}
 Jump Alpha {type=number}
]
