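# Physics asset for the Gimmick Female pickaxe, a Battle Royale cosmetic melee weapon.
# A single body on the Melee_TwoHanded_Handle bone uses one large, offset and rotated capsule as a coarse collision shape.
# It has no constraints, so it defines no jointed or independently simulated parts.
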PhysicsAsset /BRCosmetics/Weapons/FORT_Melee/Pickaxe_Gimmick_Female/Meshes/Gimmick_Female_PhysicsAsset
bodies {Melee_TwoHanded_Handle [{shapeType=Capsule, length=78.8063, radius=38.6048, center=<1.3567 -1.9291 32.7113>, rotation=<rot -0.0276 -90 0.0131>}]}
constraints []
